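# Sound cue for equipping the Chrono Cloak item, playing a single wave with no randomization or modulation.
# It routes through the gadgets and items sound class and borrows weapon reload attenuation for spatial falloff.
# It references a shared pickup concurrency set and also holds override values that would allow one instance and block new ones.
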SoundCue /ChronoCloak/SFX/Chrono_Cloak_Equip_Cue
root=WavePlayer_1
nodes [{id=WavePlayer_1, properties {soundWaveAssetPtr=&/Chrono_Cloak_Equip_01}}]
properties {
 attenuationSettings=&Fort_Weapon_Reload_Att
 maxDistance=2250
 soundClassObject=&Fort_Gadgets_Items
 volumeMultiplier=0.8
 concurrencyOverrides {maxCount=1, resolutionRule=PreventNew}
 concurrencySet [&PickupMaximum_Oldest]
}
